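# Placeholder sound cue for the game start of the Battle Pass S22 narrative test chamber minigame.
# A single wave player reuses the TimedTrial final-seconds quest sound, boosted in volume and routed to the Fort quests sound class.
# Its concurrency override allows one instance at a time and blocks new plays while it is active.
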SoundCue /BattlePassS22/QuestActors/Narrative_Pt1/TestChamber_Game/PlaceholderAudio/Proto_GameStart_Cue
root=WavePlayer_1
nodes [{id=WavePlayer_1, properties {soundWaveAssetPtr=/Game/Athena/Sounds/Quests/TimedTrial/TimedTrial_FinalSeconds}}]
properties {
 bOverrideConcurrency=true
 maxDistance=2097152
 soundClassObject=&Fort_Quests
 volumeMultiplier=2.24
 concurrencyOverrides {maxCount=1, resolutionRule=PreventNew}
}
