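# Placeholder sound cue for an incorrect button press in the Season 22 narrative quest TestChamber game.
# A single wave player plays the reused TimedTrial failure sound with boosted volume, routed through the Fort quests sound class.
# A concurrency override limits it to one playing instance at a time, and its max distance is set extremely large.
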SoundCue /BattlePassS22/QuestActors/Narrative_Pt1/TestChamber_Game/PlaceholderAudio/Proto_IncorrectButton_Cue
root=WavePlayer_0
nodes [{id=WavePlayer_0, properties {soundWaveAssetPtr=/Game/Athena/Sounds/Quests/TimedTrial/TimedTrial_Failure}}]
properties {
 bOverrideConcurrency=true
 maxDistance=2097152
 soundClassObject=&Fort_Quests
 volumeMultiplier=1.68
 concurrencyOverrides {maxCount=1}
}
